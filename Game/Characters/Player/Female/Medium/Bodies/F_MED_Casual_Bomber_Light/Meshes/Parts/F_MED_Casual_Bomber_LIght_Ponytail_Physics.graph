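# Physics asset for the ponytail on the medium female player's light casual bomber body, with capsule bodies for the hair, head and upper spine.
# Four dyn_pony bones chain from dyn_parent_pony with per-joint swing limits and tight twist between segments, looser at the root.
# The spine_05 to head and head to dyn_parent_pony joints have free swing and twist.
# Larger head and spine_05 capsules provide body volume around the hair chain.
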PhysicsAsset /Game/Characters/Player/Female/Medium/Bodies/F_MED_Casual_Bomber_Light/Meshes/Parts/F_MED_Casual_Bomber_LIght_Ponytail_Physics
bodies {
 dyn_parent_pony [{shapeType=Capsule, length=0.505, radius=0.505, rotation=<rot 0 -90 0>}]
 dyn_pony_01 [{shapeType=Capsule, length=5, radius=3, center=<0.1 0 0>, rotation=<rot 90 0 0>}]
 dyn_pony_02 [{shapeType=Capsule, length=5, radius=4.25, center=<1 0 0>, rotation=<rot 65.3776 -180 -180>}]
 dyn_pony_03 [{shapeType=Capsule, length=5, radius=3.75, center=<1.5 0 0>, rotation=<rot 90 0 0>}]
 dyn_pony_04 [{shapeType=Capsule, length=2.75, radius=3, center=<1.6427 -0.0001 0.5398>, rotation=<rot 39.99 -180 -180>}]
 head [{shapeType=Capsule, length=6.75, radius=7.5, center=<3.5252 0.0924 0>, rotation=<rot -90 0 -15.0002>}]
 spine_05 [
  {shapeType=Capsule, length=11.505, radius=14.005, center=<4.0747 5.6132 0>, rotation=<rot 0 -90 0>}
  {shapeType=Capsule, length=15, radius=15, center=<-1.1259 -1.6922 0>}
 ]
}
constraints [
 {bone1Name=dyn_pony_02, bone2Name=dyn_pony_01, swing1LimitDegrees=25, swing2LimitDegrees=109.2857, twistLimitDegrees=3}
 {bone1Name=dyn_pony_03, bone2Name=dyn_pony_02, swing1LimitDegrees=35, swing2LimitDegrees=15, twistLimitDegrees=3}
 {bone1Name=dyn_pony_04, bone2Name=dyn_pony_03, swing1LimitDegrees=35, swing2LimitDegrees=10, twistLimitDegrees=3}
 {bone1Name=dyn_pony_01, bone2Name=dyn_parent_pony, swing1LimitDegrees=5, swing2LimitDegrees=5, twistLimitDegrees=45}
 {
  bone1Name=head
  bone2Name=spine_05
  swing1LimitDegrees=45
  swing1Motion=Free
  swing2LimitDegrees=45
  swing2Motion=Free
  twistLimitDegrees=45
  twistMotion=Free
 }
 {
  bone1Name=dyn_parent_pony
  bone2Name=head
  swing1LimitDegrees=45
  swing1Motion=Free
  swing2LimitDegrees=45
  swing2Motion=Free
  twistLimitDegrees=45
  twistMotion=Free
 }
]
